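; Debug view model for the Watch Live modal in the tournament leaderboard, supplying placeholder player stats.
; Getters for average eliminations, average placement, elimination points, placement points and Victory Royales each return a random number within a fixed range.
; Functions for live slot items, average points, max points and top 50 points have empty bodies.
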(fn CreateLiveSlotsItems)

(fn GetAvgElims
  (return (Math|Random|RandomFloatinRange 5.0 4e+01)))

(fn GetAvgPlacement
  (return (Math|Random|RandomFloatinRange 5.0 4e+01)))

(fn GetAvgPoints)

(fn GetEliminationsPTS
  (return (Math|Random|RandomFloatinRange 1e+02 5e+02)))

(fn GetMaxPoints)

(fn GetPlacementPTS
  (return (Math|Random|RandomFloatinRange 1e+02 5e+02)))

(fn GetTop50Points)

(fn GetVictoryRoyales
  (return (Math|Random|RandomIntegerinRange 5 40)))
