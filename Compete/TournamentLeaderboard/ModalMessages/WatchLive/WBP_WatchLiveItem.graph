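# List entry for the tournament leaderboard's Watch Live modal, showing a time and player names.
# List item binding casts to VM_WatchLiveItemDebug and stores its entry view model.
# SetTime fills the time block; SetPlayersInfo joins names with "&&", adding rich-text flags except for "fortnite".
# Selection animates player-text material colors, reversing on deselect at UIKit timing speed; mobile enlarges the height spacer.
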WidgetBlueprint /Compete/TournamentLeaderboard/ModalMessages/WatchLive/WBP_WatchLiveItem
animations [Anim_TransitionSelected {
  binding_count=1
  length_seconds=1
  object=&self:Anim_TransitionSelected
  sequence {bindings [CommonRichText_PlayersInfo {
     id=#1
     possessed_class=CommonRichTextBlock
     tracks [{
       type=MovieSceneWidgetMaterialTrack
       display_name=Default Text Style Override.Font.FontMaterial
       sections [{
         type=MovieSceneParameterSection
         end_frame=20
         start_frame=0
         channels [
          Trans_Focused {type=MovieSceneScriptingFloatChannel, const [0,2]}
          Trans_Hovered {type=MovieSceneScriptingFloatChannel, const [0,2]}
          Trans_Pressed {type=MovieSceneScriptingFloatChannel, const [0,2]}
          Color_IdleR {type=MovieSceneScriptingFloatChannel, keys [[0,1],[20,0]]}
          Color_IdleG {type=MovieSceneScriptingFloatChannel, keys [[0,1],[20,0]]}
          Color_IdleB {type=MovieSceneScriptingFloatChannel, keys [[0,1],[20,0]]}
          Color_IdleA {type=MovieSceneScriptingFloatChannel, const [1,2]}
          Color_HoveredR {type=MovieSceneScriptingFloatChannel, keys [[0,0.0103],[20,0]]}
          Color_HoveredG {type=MovieSceneScriptingFloatChannel, keys [[0,0.0103],[20,0]]}
          Color_HoveredB {type=MovieSceneScriptingFloatChannel, keys [[0,0.0103],[20,0]]}
          Color_HoveredA {type=MovieSceneScriptingFloatChannel, const [1,2]}
          Color_FocusedR {type=MovieSceneScriptingFloatChannel, keys [[0,0.0103],[20,0]]}
          Color_FocusedG {type=MovieSceneScriptingFloatChannel, keys [[0,0.0103],[20,0]]}
          Color_FocusedB {type=MovieSceneScriptingFloatChannel, keys [[0,0.0103],[20,0]]}
          Color_FocusedA {type=MovieSceneScriptingFloatChannel, const [1,2]}
          Color_PressedR {type=MovieSceneScriptingFloatChannel, keys [[0,0.0103],[20,0]]}
          Color_PressedG {type=MovieSceneScriptingFloatChannel, keys [[0,0.0103],[20,0]]}
          Color_PressedB {type=MovieSceneScriptingFloatChannel, keys [[0,0.0103],[20,0]]}
          Color_PressedA {type=MovieSceneScriptingFloatChannel, const [1,2]}
         ]
        }]
      }]
    }]}
 }]
graphs [
 EventGraph {nodes [
   n1 = Event PreConstruct()
   > Parent: PreConstruct(IsDesignTime=@n1.IsDesignTime)
   > RegisterModularBlocks()
   > SetMobileStyle() {is=title K2Node_CallFunction}
   Event On Selected()
   > Parent: On Selected() {is=title K2Node_CallParentFunction}
   n2 = Get CommonRichText_PlayersInfo() -> (CommonRichText_PlayersInfo:Common Rich Text Block Object Reference)
   > SetDefaultMaterial!(
    execute
    self:Rich Text Block Object Reference=@n2
    InMaterial:Material Interface Object Reference=/FortUILibrary/UIKit/Buttons/MaterialInstances/SelectableButtons/MI_UIKit_SelectableText_Quiet
   ) -> (then)
   > GetDefaultDynamicMaterial!(execute, self:Rich Text Block Object Reference=@n2) -> (then, ReturnValue:Material Instance Dynamic Object Reference)
   n3 = float / float(
    A=1
    B=Break UIKit Block Timing!(UIKitBlockTiming:"UIKit Block Timing Structure (by ref)"=GetBlockTimings!() -> (BlockTiming:UIKit Block Timing Structure)) -> (Selecting:float=~) {is=title K2Node_BreakStruct}
   )
   > PlayAnimation(InAnimation=Get Anim_TransitionSelected() -> (Anim_TransitionSelected:Widget Animation Object Reference), PlaybackSpeed=@n3)
   Event On Deselected()
   > Parent: On Deselected() {is=title K2Node_CallParentFunction}
   > PlayAnimation(InAnimation=Get Anim_TransitionSelected(), PlayMode=Reverse, PlaybackSpeed=@n3)
   n4 = Event OnListItemObjectSet()
   > n5 = Cast To VM_WatchLiveItemDebug!(execute, Object=@n4.ListItemObject) -> (then, CastFailed, AsVM Watch Live Item Debug:VM Watch Live Item Debug Object Reference) {is=title K2Node_DynamicCast}
   > Set VM_WatchLiveItemDebug(VM_WatchLiveItemDebug:VM Watch Live Item Debug Object Reference=@n5) -> (Output_Get:VM Watch Live Item Debug Object Reference)
   > Set FortTournamentWatchLiveEntryVM(FortTournamentWatchLiveEntryVM:Watch Live Entry VM Object Reference=Get EntryVM!(self:VM Watch Live Item Debug Object Reference=@n5) -> (EntryVM:Watch Live Entry VM Object Reference)) -> (Output_Get:Watch Live Entry VM Object Reference)
  ]}
 RegisterModularBlocks {nodes [
   RegisterModularBlocks() {is=title K2Node_FunctionEntry}
   > SetFont!(
    execute
    self:WBP UIKit Block Text Object Reference=Get BlockText_Time() -> (BlockText_Time:WBP UIKit Block Text Object Reference)
    Font:"Slate Font Info Structure (by ref)"=Get BlockTextFont() -> (BlockTextFont:Slate Font Info Structure)
   ) -> (then)
   > SetFontMaterial!(
    execute
    self:WBP UIKit Block Text Object Reference=Get BlockText_Time()
    In Material:Material Instance Object Reference=&MI_UIKit_SelectableText_Quiet
   ) -> (then)
   > SetText!(execute, self:WBP UIKit Block Text Object Reference=Get BlockText_Time(), Text:Text="LOCTEXT(\"00:00\")") -> (then)
   > RegisterModularBlock!(execute, self:UIKit Modular Button Object Reference, Block:UIKit Block Object Reference=Get BlockText_Time()) -> (then)
  ]}
 SetMobileStyle {nodes [
   SetMobileStyle() {is=title K2Node_FunctionEntry}
   > Branch(Condition=IsMobileGame())
   > SetSize!(execute, self:Spacer Object Reference=Get Spacer_Height() -> (Spacer_Height:Spacer Object Reference), InSize:Vector 2D Structure=(X=1,Y=80)) -> (then)
  ]}
 SetPlayersInfo {nodes [
   n1 = SetPlayersInfo!() -> (then, Players Profiles:Map of Strings to Tournament Player Profile VM Object References) {is=title K2Node_FunctionEntry}
   > n2 = Keys!(execute, TargetMap:Map of Strings to Tournament Player Profile VM Object References=@n1) -> (then, Keys:Array of Strings)
   > n3 = For Each Loop(Array=@n2)
   n4 = Find(TargetMap:Map of Strings to Tournament Player Profile VM Object References=@n1, Key=@n3) -> (Value:Tournament Player Profile VM Object Reference)
   n5 = Get DisplayName!(self:Tournament Player Profile VM Object Reference=@n4) -> (DisplayName:String)
   n6 = Get FlagName!(self:Tournament Player Profile VM Object Reference=@n4) -> (FlagName:String)
   n7 = "To String (Text)"(InText=Select(
     Option 0:"Text (by ref)"=Format Text!(Format="LOCTEXT(\"<img id=\\\"{Country}\\\" size=\\\"18,18\\\" height=\\\"25\\\"/> {Nickname}\")", Country:String=@n6, Nickname:String=@n5) -> (Result)
     Option 1:"Text (by ref)"="To Text (String)"(InString=@n5)
     Index="Equal Exactly (String)"(A=@n6, B=fortnite)
    ) -> (ReturnValue:"Text (by ref)"))
   n8 = Get PlayersInfo() -> (PlayersInfo:String)
   n9 = "Equal (Integer)"(A=@n3.Array Index, B=Last Index!(TargetArray=@n2) -> (ReturnValue=0))
   > Set PlayersInfo(PlayersInfo:String=Append(A=Append(A=@n8, B=@n7), B=Select(Option 0:String=" && ", Option 1:String, Index=@n9) -> (ReturnValue:String))) -> (Output_Get:String)
   SetText!(
    execute=@n3.Completed
    self:Rich Text Block Object Reference=Get CommonRichText_PlayersInfo()
    InText:"Text (by ref)"="To Text (String)"(InString=Get PlayersInfo())
   ) -> (then)
  ]}
 SetTime {nodes [
   n1 = SetTime!() -> (then, InText:Text) {is=title K2Node_FunctionEntry}
   > SetText!(execute, self:WBP UIKit Block Text Object Reference=Get BlockText_Time(), Text:Text=@n1) -> (then)
  ]}
]
variables [BlockTextFont {
  type=SlateFontInfo
  default {
   bForceMonospaced=false
   bMaterialIsStencil=false
   fontMaterial=null
   fontObject=&HeadingNow
   letterSpacing=0
   monospacedWidth=1
   size=18
   skewAmount=0
   typefaceFontName=75 Medium
   outlineSettings {
    bApplyOutlineToDropShadows=false
    bMiteredCorners=false
    bSeparateFillAlpha=false
    outlineMaterial=null
    outlineSize=0
    outlineColor=<rgba 0 0 0 1>
   }
  }
 }]
widgets {
 info {
  inheritedWidgetCount=0
  namedSlotCount=1
  parentClass=/Compete/TournamentLeaderboard/ModalMessages/WBP_LeaderboardListEntry_Base.WBP_LeaderboardListEntry_Base_C
  rootWidgetClass=null
  widgetCount=4
 }
 widgets [
  Main {widgetClassPath=&GridPanel, widget=Main, properties {columnFill [0,1], rowFill [1]}}
  BlockText_Time {
   widgetClassPath=/FortUILibrary/Systems/ModularButton/Blocks/WBP_UIKit_Block_Text
   bIsVariable=true
   parent=Main
   widget=BlockText_Time
   slot {class=GridSlot, properties {horizontalAlignment=HAlign_Left, layer=1, verticalAlignment=VAlign_Center, padding=<ltrb 32 0 0 0>}}
  }
  CommonRichText_PlayersInfo {
   widgetClassPath=&CommonRichTextBlock
   bIsVariable=true
   parent=Main
   widget=CommonRichText_PlayersInfo
   properties {
    autoWrapText=true
    bIsScrollingEnabled=false
    bOverrideDefaultStyle=true
    mobileTextBlockScale=1.14
    text="WWWWWWWWWWWWWWWW && WWWWWWWWWWWWWWWW  "
    decoratorClasses [/Compete/DataTables/BP_FlagsDecorator.BP_FlagsDecorator_C]
    defaultTextStyleOverride {
     overflowPolicy=Clip
     transformPolicy=None
     colorAndOpacity {colorUseRule=UseColor_Specified, specifiedColor=<rgba 0.5765 0.5765 0.5765 0.8>}
     font {
      bForceMonospaced=false
      bMaterialIsStencil=false
      fontMaterial=&MI_UIKit_SelectableText_Quiet
      fontObject=&HeadingNow
      letterSpacing=0
      monospacedWidth=1
      size=15.75
      skewAmount=0
      typefaceFontName=75 Medium
      outlineSettings {
       bApplyOutlineToDropShadows=false
       bMiteredCorners=false
       bSeparateFillAlpha=false
       outlineMaterial=null
       outlineSize=0
       outlineColor=<rgba 0 0 0 1>
      }
     }
     highlightColor {colorUseRule=UseColor_Specified, specifiedColor=<rgba 0 0 0 1>}
     highlightShape {
      bIsDynamicallyLoaded=false
      drawAs=Image
      imageType=NoImage
      mirroring=NoMirror
      resourceName=None
      resourceObject=null
      tiling=NoTile
      imageSize=<32 32>
      margin=<ltrb 0 0 0 0>
      outlineSettings {
       bUseBrushTransparency=false
       roundingType=HalfHeightRadius
       width=0
       color {colorUseRule=UseColor_Specified, specifiedColor=<rgba 0 0 0 0>}
       cornerRadii=<0 0 0 0>
      }
      tintColor {colorUseRule=UseColor_Specified, specifiedColor=<rgba 1 1 1 1>}
      uVRegion {bIsValid=false, max=<0 0>, min=<0 0>}
     }
     selectedBackgroundColor {colorUseRule=UseColor_Foreground, specifiedColor=<rgba 1 0 1 1>}
     shadowColorAndOpacity=<rgba 0 0 0 1>
     shadowOffset=<0 0>
     strikeBrush {
      bIsDynamicallyLoaded=false
      drawAs=Image
      imageType=NoImage
      mirroring=NoMirror
      resourceName=None
      resourceObject=null
      tiling=NoTile
      imageSize=<32 32>
      margin=<ltrb 0 0 0 0>
      outlineSettings {
       bUseBrushTransparency=false
       roundingType=HalfHeightRadius
       width=0
       color {colorUseRule=UseColor_Specified, specifiedColor=<rgba 0 0 0 0>}
       cornerRadii=<0 0 0 0>
      }
      tintColor {colorUseRule=UseColor_Specified, specifiedColor=<rgba 1 1 1 1>}
      uVRegion {bIsValid=false, max=<0 0>, min=<0 0>}
     }
     underlineBrush {
      bIsDynamicallyLoaded=false
      drawAs=Image
      imageType=NoImage
      mirroring=NoMirror
      resourceName=None
      resourceObject=null
      tiling=NoTile
      imageSize=<32 32>
      margin=<ltrb 0 0 0 0>
      outlineSettings {
       bUseBrushTransparency=false
       roundingType=HalfHeightRadius
       width=0
       color {colorUseRule=UseColor_Specified, specifiedColor=<rgba 0 0 0 0>}
       cornerRadii=<0 0 0 0>
      }
      tintColor {colorUseRule=UseColor_Specified, specifiedColor=<rgba 1 1 1 1>}
      uVRegion {bIsValid=false, max=<0 0>, min=<0 0>}
     }
    }
   }
   slot {class=GridSlot, properties {columnSpan=2, layer=1, verticalAlignment=VAlign_Center, padding=<ltrb 146 0 32 0>}}
  }
  Spacer_Height {
   widgetClassPath=&Spacer
   bIsVariable=true
   parent=Main
   widget=Spacer_Height
   properties {size=<1 74>}
   slot {class=GridSlot, properties {horizontalAlignment=HAlign_Left}}
  }
 ]
}
