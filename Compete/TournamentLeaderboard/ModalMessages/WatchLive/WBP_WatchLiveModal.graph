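# Modal dialog in the tournament leaderboard that lists live games and starts watching the selected one through the Watch Live viewmodel.
# Initialize passes category and friends filter to the viewmodel, though both header controls are disabled.
# SetLiveGames wraps entries in VM_WatchLiveItemDebug objects for the list, or shows an empty message and hides Watch.
# Scrolling near the list's end requests more games, and switching to gamepad refreshes focus.
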WidgetBlueprint /Compete/TournamentLeaderboard/ModalMessages/WatchLive/WBP_WatchLiveModal
graphs [
 BP_GetDesiredFocusTarget {nodes [
   Get Desired Focus Target()
   > Return Node!(execute, ReturnValue:Widget Object Reference=Get ListView_LiveGames() -> (ListView_LiveGames:Common List View Object Reference))
  ]}
 EventGraph {nodes [
   n1 = "On Clicked (WatchButton)"!() -> (OutputDelegate, then, Button:Common Button Base Object Reference) {is=title K2Node_ComponentBoundEvent}
   > Branch(Condition=Get bIsActive() -> (bIsActive:bool))
   > n2 = Cast To FortTournamentWatchLiveEntryVM!(
    execute
    Object=Get Selected Item!(self:List View Object Reference=Get ListView_LiveGames()) -> (ReturnValue:Object Reference) {is=title K2Node_CallFunction}
   ) -> (then, CastFailed, AsWatch Live Entry VM:Watch Live Entry VM Object Reference) {is=title K2Node_DynamicCast}
   > Watch!(
    execute
    self:Watch Live VM Object Reference=Get FortTournamentWatchLiveVM() -> (FortTournamentWatchLiveVM:Watch Live VM Object Reference)
    LiveSessionId:String=Get LiveSessionId!(self:Fort Tournament Base Entry VM Object Reference=@n2) -> (LiveSessionId:String)
   ) -> (then)
   > Fire Event Element Interacted(Element=@n1.Button)
   n3 = "On Clicked (CloseButton)"!() -> (OutputDelegate, then, Button:Common Button Base Object Reference) {is=title K2Node_ComponentBoundEvent}
   > Branch(Condition=Get bIsActive())
   > DeactivateWidget()
   > Fire Event Element Interacted(Element=@n3.Button)
   n4 = "On Item Selection Changed (ListView_LiveGames)"!() -> (OutputDelegate, then, Item:Object Reference, bIsSelected:bool) {is=title K2Node_ComponentBoundEvent}
   > Branch(Condition=@n4.bIsSelected)
   > Scroll Item Into View!(execute, self:List View Object Reference=Get ListView_LiveGames(), Item:Object Reference=@n4.Item) -> (then)
   n5 = "On Item Scrolled Into View (ListView_LiveGames)"!() -> (OutputDelegate, then, Item:Object Reference, Widget:User Widget Object Reference) {is=title K2Node_ComponentBoundEvent}
   n6 = GetListItems!(self:List View Object Reference=Get ListView_LiveGames()) -> (ReturnValue:Array of Object References)
   n7 = Find Item!(TargetArray:Array of Object References=@n6, ItemToFind:"Object Reference (by ref)"=@n5.Item) -> (ReturnValue)
   n8 = Last Index!(TargetArray:Array of Object References=@n6) -> (ReturnValue=0)
   > n9 = Branch(Condition="float >= float"(A=float / float(A:int=@n7, B:int=@n8), B=0.9))
   Branch(
    execute=[@n9.then, @n11.then]
    Condition=CanRequestMoreLiveGames!(self:Watch Live VM Object Reference=Get FortTournamentWatchLiveVM()) -> (ReturnValue:bool)
   )
   > ShowLoadingEntriesMessage!(execute, self, bShow:bool=true) -> (then)
   > RequestMoreLiveGames(self:Watch Live VM Object Reference=Get FortTournamentWatchLiveVM())
   n10 = "On List View Scrolled (ListView_LiveGames)"!() -> (OutputDelegate, then, ItemOffset:float, DistanceRemaining:float) {is=title K2Node_ComponentBoundEvent}
   > n11 = Branch(Condition="float <= float"(A=@n10.DistanceRemaining, B=Get Min Distance Remaining For Next Page Request() -> (Min Distance Remaining For Next Page Request:double)) -> (ReturnValue=~))
   n12 = OnInputMethodChanged!() -> (OutputDelegate, then, bNewInputType:ECommonInputType Enum) {is=title K2Node_CustomEvent}
   > n13 = Switch on ECommonInputType!(execute, Selection:ECommonInputType Enum=@n12.bNewInputType) -> (MouseAndKeyboard:Exec, Gamepad:Exec, Touch:Exec, Count:Exec) {is=title K2Node_SwitchEnum}
   UpdateFocus(execute=@n13.Gamepad)
   n14 = ShowLoadingEntriesMessage!() -> (OutputDelegate, then, bShow:bool=~) {is=title K2Node_CustomEvent}
   > SetVisibility(
    self=Get LoadingEntriesMessage() -> (LoadingEntriesMessage:WBP Loading Entries Message Object Reference)
    InVisibility=Select(Option 0=Collapsed, Option 1=HitTestInvisible, Index=@n14.bShow)
   )
   Event PreConstruct()
   n15 = Get Dialog_Base() -> (Dialog_Base:WBP UIKit Dialog Internal Object Reference)
   > SetPadding!(
    execute
    self:Grid Slot Object Reference=SlotAsGridSlot(Widget=Get Slot_Content!(self:WBP UIKit Dialog Internal Object Reference=@n15) -> (Slot_Content:Named Slot Object Reference))
    InPadding_Left:float=65
    InPadding_Top:float=0
    InPadding_Right:float=80
    InPadding_Bottom:float=0
   ) -> (then)
   > n16 = Is Valid(InputObject=Get ContextMenuActionContext() -> (ContextMenuActionContext:UIKit Context Menu Action Context Object Reference))
   n17 = Construct Object from Class(execute=@n16.Is Not Valid, Class=/Script/FortniteUI.UIKitContextMenuActionContext) -> (ReturnValue:UIKit Context Menu Action Context Object Reference)
   > Set ContextMenuActionContext(ContextMenuActionContext:UIKit Context Menu Action Context Object Reference=@n17) -> (Output_Get:UIKit Context Menu Action Context Object Reference)
   Event Construct()
   > Fire Event Sub View Opened!(execute, View:Widget Object Reference=Self-Reference(), AdditionalInfo:Flow Analytics Additional Info Structure, bIsOverlay:bool=true) -> (then)
   Event On Activated()
   > Assign On Input Method Changed!(execute:Exec, self:Common Input Object Reference=Get CommonInputSubsystem(), Delegate:"Delegate (by ref)"=@n12) -> (then:Exec) {is=title K2Node_AssignDelegate}
   > OnInputMethodChanged!(execute, self, bNewInputType:ECommonInputType Enum=GetCurrentInputType(self=Get CommonInputSubsystem())) -> (then) {is=title K2Node_CallFunction}
   Event On Deactivated()
   > Unbind Event from On Input Method Changed(self=Get CommonInputSubsystem(), Delegate=Create Event())
  ]}
 Initialize {nodes [
   n1 = Initialize!() -> (
    then
    WatchLiveVM:Watch Live VM Object Reference
    LeaderboardCategory:"Fort Compete UILeaderboard Category Structure (by ref)"
    bShowFriendsOnly:bool=~
    AlternativeButtonVisibility:ESlateVisibility Enum=Visible
   ) {is=title K2Node_FunctionEntry}
   > SetFortTournamentWatchLiveVM!(execute, self, Viewmodel:Watch Live VM Object Reference=@n1) -> (then)
   > Set CurrentCategory(CurrentCategory:Fort Compete UILeaderboard Category Structure=@n1.LeaderboardCategory) -> (Output_Get:Fort Compete UILeaderboard Category Structure)
   > InitializeVM!(
    execute
    self:Watch Live VM Object Reference=Get FortTournamentWatchLiveVM()
    InLeaderboardCategory:Fort Compete UILeaderboard Category Structure=@n1.LeaderboardCategory
    bInShowFriendsOnly:bool=@n1.bShowFriendsOnly
   ) -> (then)
   > SetVisibility(
    self=Get Alternative_Button() -> (Alternative_Button:WBP UIKit Button Regular Object Reference)
    InVisibility=Get AlternativeButtonVisibility() -> (AlternativeButtonVisibility:ESlateVisibility Enum=Visible)
   )
   > Set Text(
    Text=Break Fort Compete UILeaderboard Category!(FortCompeteUILeaderboardCategory:"Fort Compete UILeaderboard Category Structure (by ref)"=@n1.LeaderboardCategory) -> (LeaderboardName:Text, GroupingKey:String, InstanceId:String) {is=title K2Node_BreakStruct}
    self:Fort CTAButton Object Reference=Get Alternative_Button()
   )
   > SetIsSelected!(
    execute
    self:Common Button Base Object Reference=Get FriendsSwitcher() -> (FriendsSwitcher:WBP UIKit Switcher Button Object Reference)
    InSelected:bool=Get bShowFriendsOnly() -> (bShowFriendsOnly:bool)
    bGiveClickFeedback:bool=false
   ) -> (then)
  ]}
 SetLiveGames {nodes [
   n1 = SetLiveGames!() -> (then, LiveGames:Array of Watch Live Entry VM Object References) {is=title K2Node_FunctionEntry}
   n2 = Is Empty!(TargetArray:Array of Watch Live Entry VM Object References=@n1) -> (ReturnValue:bool)
   > SetActiveWidgetIndex(
    self:Common Visibility Switcher Object Reference=Get ListSwitcher() -> (ListSwitcher:Common Visibility Switcher Object Reference)
    Index="To Integer (Boolean)"!(InBool:bool=@n2) -> (ReturnValue:int=0)
   )
   > SetVisibility(
    self=Get WatchButton() -> (WatchButton:WBP UIKit Button Regular Object Reference)
    InVisibility=Select(Option 0=Visible, Option 1=Collapsed, Index=@n2)
   )
   > SetWatchAvailable!(execute, self, Avaliable:bool=NOT Boolean(A=@n2)) -> (then)
   > n3 = Branch(Condition=@n2)
   > UpdateFocus()
   Clear(
    execute=@n3.else
    TargetArray:Array of VM Watch Live Item Debug Object References=Get WatchLiveVMs() -> (WatchLiveVMs:Array of VM Watch Live Item Debug Object References)
   )
   > ClearListItems(self=Get ListView_LiveGames())
   > n4 = For Each Loop(Array:Array of Watch Live Entry VM Object References=Get LiveGames() -> (LiveGames:Array of Watch Live Entry VM Object References)) -> (Array Element:Watch Live Entry VM Object Reference)
   Set List Items!(execute=@n4.Completed, self:List View Object Reference=Get ListView_LiveGames(), InListItems:Array of Object References=Get WatchLiveVMs()) -> (then) {is=title K2Node_CallFunction}
   > n5 = Sequence!(execute) -> (then_0, then_1, then_2:Exec)
   n6 = Length(TargetArray:Array of Watch Live Entry VM Object References=Get LiveGames())
   n7 = float / float(A:double="To Float (Integer)"(InInt=10), B:double="To Float (Integer)"(InInt=@n6))
   > Set Min Distance Remaining For Next Page Request(Min Distance Remaining For Next Page Request:double=Select(
     Option 0:double="Min (Float)"(A:double=@n7, B:double=0.1) -> (ReturnValue:double) {is=title K2Node_CommutativeAssociativeBinaryOperator}
     Option 1:double=0
     Index="Equal (Integer)"(A=@n6)
    ) -> (ReturnValue:double)) -> (Output_Get:double)
   n8 = Get FortTournamentWatchLiveVM()
   ShowLoadingEntriesMessage(execute=@n5.then_1, bShow=NOT Boolean(A=Get bIsModelInitialized!(self:Watch Live VM Object Reference=@n8) -> (bIsModelInitialized:bool))) {is=title K2Node_CallFunction}
   SetSelectedIndex!(execute=@n5.then_2, self:List View Object Reference=Get ListView_LiveGames(), Index:int=0) -> (then)
   > UpdateFocus()
   n9 = Construct Object from Class!(
    execute=@n4.LoopBody
    Class=&/VM_WatchLiveItemDebug.VM_WatchLiveItemDebug_C
    self
    Index:int=@n4.Array Index
    EntryVM:Watch Live Entry VM Object Reference=@n4
   ) -> (then, ReturnValue:VM Watch Live Item Debug Object Reference)
   > Add(
    TargetArray:Array of VM Watch Live Item Debug Object References=Get WatchLiveVMs()
    NewItem:"VM Watch Live Item Debug Object Reference (by ref)"=@n9
   ) {is=title K2Node_CallArrayFunction}
  ]}
 SetTitle {nodes [
   n1 = SetTitle!() -> (then, Title:Text) {is=title K2Node_FunctionEntry}
   > SetDialogTitle!(execute, self:WBP UIKit Dialog Internal Object Reference=Get Dialog_Base(), Title:Text=@n1) -> (then)
  ]}
 SetWatchAvailable {nodes [
   n1 = SetWatchAvailable!() -> (then, Avaliable:bool=~) {is=title K2Node_FunctionEntry}
   > SetVisibility(
    self=Get WatchButton()
    InVisibility=Select(Option 0:ESlateVisibility Enum=Collapsed, Option 1:ESlateVisibility Enum=Visible, Index=@n1) -> (ReturnValue:ESlateVisibility Enum)
   )
  ]}
 UpdateFocus {nodes [
   UpdateFocus() {is=title K2Node_FunctionEntry}
   n1 = GetActiveWidgetIndex!(self:Common Visibility Switcher Object Reference=Get ListSwitcher()) -> (ReturnValue:int=0)
   > SetFocus(self=Select(
     Option 0:Widget Object Reference=Get Dialog_Base()
     Option 1:Widget Object Reference=Get ListView_LiveGames()
     Index="Equal (Integer)"(A=@n1)
    ) -> (ReturnValue:Widget Object Reference))
  ]}
]
variables [
 Categories {type="array<FortCompeteUILeaderboardCategory>", default []}
 ContextMenuActionContext {type=UIKitContextMenuActionContext, default=null}
 CurrentCategory {type=FortCompeteUILeaderboardCategory, default {groupingKey="", instanceId="", leaderboardName=""}}
 Min Distance Remaining For Next Page Request {type=number}
]
widgets {
 info {
  inheritedWidgetCount=0
  namedSlotCount=1
  parentClass=&CommonActivatableWidget
  rootWidgetClass=/FortUILibrary/Systems/Dialog/WBP_UIKit_Dialog_Internal.WBP_UIKit_Dialog_Internal_C
  widgetCount=18
 }
 widgets [
  Dialog_Base {
   widgetClassPath=/FortUILibrary/Systems/Dialog/WBP_UIKit_Dialog_Internal
   bIsVariable=true
   widget=Dialog_Base
   properties {
    bIsFocusable=true
    useText=None
    buttons {
     each Button Horizontal Fill=true
     each Button Min Width=0
     each Button Size=Medium
     focuse Button=0
     space Between Buttons=24
     cTA Buttons Types [Quiet]
     group Placement {
      horizontal Alignment=HAlign_Fill
      vertical Alignment=VAlign_Bottom
      grid {column=1, column Span=0, layer=0, row=2, row Span=0, nudge=<0 0>}
      padding=<ltrb 80 64 80 74>
     }
    }
    dialogSIze {
     height Override {enable=false, value=0}
     max Aspect Ratio {enable=false, value=0}
     max Desired Height {enable=true, value=932}
     max Desired Width {enable=false, value=0}
     min Aspect Ratio {enable=false, value=0}
     min Desired Height {enable=true, value=592}
     min Desired Width {enable=false, value=0}
     width Override {enable=true, value=1132}
    }
    imageProperties {
     material=/FortUILibrary/Systems/Dialog/MaterialInstances/MI_UI_DialogImage
     image Size=<0 0>
     placement {
      horizontal Alignment=HAlign_Fill
      vertical Alignment=VAlign_Fill
      grid {column=1, column Span=0, layer=20, row=0, row Span=1, nudge=<0 0>}
      padding=<ltrb 0 0 0 0>
     }
    }
    titleProperties {
     auto Wrap Text=true
     justification=Left
     line Height Percentage=1
     mobile Font Size Multiplier=1.3
     transform Policy=ToUpper
     wrap Text At=0
     color {colorUseRule=UseColor_Specified, specifiedColor=<rgba 1 1 1 1>}
     font {
      bForceMonospaced=false
      bMaterialIsStencil=false
      fontMaterial=null
      fontObject=&HeadingNow
      letterSpacing=0
      monospacedWidth=1
      size=25.5
      skewAmount=0
      typefaceFontName=86 Bold Italic
      outlineSettings {
       bApplyOutlineToDropShadows=false
       bMiteredCorners=false
       bSeparateFillAlpha=false
       outlineMaterial=null
       outlineSize=0
       outlineColor=<rgba 0 0 0 1>
      }
     }
     placement {
      horizontal Alignment=HAlign_Fill
      vertical Alignment=VAlign_Top
      grid {column=1, column Span=0, layer=0, row=0, row Span=0, nudge=<0 0>}
      padding=<ltrb 80 80 80 10>
     }
    }
   }
  }
  Content {widgetClassPath=&GridPanel, namedSlotHost=Dialog_Base, widget=Content, properties {columnFill [1], rowFill [0,0,1]}}
  LiveStatus {
   widgetClassPath=&GridPanel
   parent=Content
   widget=LiveStatus
   properties {columnFill [0,1], rowFill [0]}
   slot {class=GridSlot, properties {verticalAlignment=VAlign_Center, padding=<ltrb 15 0 0 0>}}
  }
  Image_LiveStatus {
   widgetClassPath=&Image
   bIsVariable=true
   parent=LiveStatus
   widget=Image_LiveStatus
   properties {brush {
     bIsDynamicallyLoaded=false
     drawAs=Image
     imageType=FullColor
     mirroring=NoMirror
     resourceName=None
     resourceObject=/Compete/MaterialInstances/MI_UI_Compete_LiveStatus
     tiling=NoTile
     imageSize=<12 12>
     margin=<ltrb 0 0 0 0>
     outlineSettings {
      bUseBrushTransparency=false
      roundingType=HalfHeightRadius
      width=0
      color {colorUseRule=UseColor_Specified, specifiedColor=<rgba 0 0 0 0>}
      cornerRadii=<0 0 0 0>
     }
     tintColor {colorUseRule=UseColor_Specified, specifiedColor=<rgba 1 1 1 1>}
     uVRegion {bIsValid=false, max=<0 0>, min=<0 0>}
    }}
   slot {class=GridSlot, properties {horizontalAlignment=HAlign_Left, verticalAlignment=VAlign_Center, padding=<ltrb 0 0 6 0>}}
  }
  CommonText_Live {
   widgetClassPath=&CommonTextBlock
   parent=LiveStatus
   widget=CommonText_Live
   properties {
    mobileFontSizeMultiplier=1.14
    scrollStyle=&TextScrollStyle-Base_C
    text=Live Games
    textTransformPolicy=ToUpper
    colorAndOpacity {colorUseRule=UseColor_Specified, specifiedColor=<rgba 0.5776 0.5776 0.5776 1>}
    font {
     bForceMonospaced=false
     bMaterialIsStencil=false
     fontMaterial=null
     fontObject=&HeadingNow
     letterSpacing=0
     monospacedWidth=1
     size=15.75
     skewAmount=0
     typefaceFontName=75 Medium
     outlineSettings {
      bApplyOutlineToDropShadows=false
      bMiteredCorners=false
      bSeparateFillAlpha=false
      outlineMaterial=null
      outlineSize=0
      outlineColor=<rgba 0 0 0 1>
     }
    }
    shadowOffset=<0 0>
   }
   slot {class=GridSlot, properties {column=1, horizontalAlignment=HAlign_Left, verticalAlignment=VAlign_Center}}
  }
  Header {
   widgetClassPath=&GridPanel
   bIsVariable=true
   parent=Content
   widget=Header
   properties {columnFill [0,0,1], rowFill [1]}
   slot {class=GridSlot, properties {layer=1, row=1, padding=<ltrb 15 32 15 0>}}
  }
  FriendsSwitcher {
   widgetClassPath=/FortUILibrary/UIKit/Buttons/WBP_UIKit_SwitcherButton
   bIsVariable=true
   parent=Header
   widget=FriendsSwitcher
   properties {
    bIsEnabled=false
    bLocked=true
    bToggleable=true
    mobileMinHeight=80
    renderOpacity=0.3
    switcherBockgroundmaterial=/Compete/MaterialInstances/MI_UI_CompeteToggleBackground
    text=Friends
    triggeringInputAction {dataTable=null, rowName=None}
   }
   slot {class=GridSlot, properties {horizontalAlignment=HAlign_Left, verticalAlignment=VAlign_Center}}
  }
  Alternative_Button {
   widgetClassPath=&WBP_UIKit_Button_Regular
   bIsVariable=true
   parent=Header
   widget=Alternative_Button
   properties {bIsEnabled=false, bLocked=true, buttonType=Quiet, minWidthOverride=303, mobileMinWidth=303, text=Default, visibility=Collapsed}
   slot {class=GridSlot, properties {column=1, padding=<ltrb 24 0 0 0>}}
  }
  ListSwitcher {
   widgetClassPath=&CommonVisibilitySwitcher
   bIsVariable=true
   parent=Content
   widget=ListSwitcher
   slot {class=GridSlot, properties {row=2, padding=<ltrb 0 0 0 64>}}
  }
  LiveItemsWrapper {
   widgetClassPath=&GridPanel
   parent=ListSwitcher
   widget=LiveItemsWrapper
   properties {clipping=ClipToBounds, columnFill [1], rowFill [1]}
   slot {class=CommonVisibilitySwitcherSlot, properties {padding=<ltrb 0 32 0 0>}}
  }
  ListView_LiveGames {
   widgetClassPath=&CommonListView
   bIsVariable=true
   parent=LiveItemsWrapper
   widget=ListView_LiveGames
   properties {
    bReturnFocusToSelection=true
    clipping=Inherit
    entryWidgetClass=&/WBP_WatchLiveItem.WBP_WatchLiveItem_C
    navigation=&self:WidgetTree.ListView_LiveGames.WidgetNavigation_0
    numDesignerPreviewEntries=6
    verticalEntrySpacing=8
    scrollBarPadding=<ltrb 15 0 0 0>
    scrollBarStyle {
     thickness=12
     draggedThumbImage {
      bIsDynamicallyLoaded=false
      drawAs=RoundedBox
      imageType=FullColor
      mirroring=NoMirror
      resourceName=../../../Engine/Content/Slate/Common/Scrollbar_Thumb.png
      resourceObject=null
      tiling=NoTile
      imageSize=<12 12>
      margin=<ltrb 0.25 0.25 0.25 0.25>
      outlineSettings {
       bUseBrushTransparency=false
       roundingType=FixedRadius
       width=2
       color {colorUseRule=UseColor_Specified, specifiedColor=<rgba 0 0 0 0>}
       cornerRadii=<6 6 6 6>
      }
      tintColor {colorUseRule=UseColor_Specified, specifiedColor=<rgba 1 1 1 1>}
      uVRegion {bIsValid=false, max=<0 0>, min=<0 0>}
     }
     horizontalBackgroundImage {
      bIsDynamicallyLoaded=false
      drawAs=NoDrawType
      imageType=NoImage
      mirroring=NoMirror
      resourceName=None
      resourceObject=null
      tiling=NoTile
      imageSize=<0 0>
      margin=<ltrb 0 0 0 0>
      outlineSettings {
       bUseBrushTransparency=false
       roundingType=HalfHeightRadius
       width=0
       color {colorUseRule=UseColor_Specified, specifiedColor=<rgba 0 0 0 0>}
       cornerRadii=<0 0 0 0>
      }
      tintColor {colorUseRule=UseColor_Specified, specifiedColor=<rgba 1 1 1 1>}
      uVRegion {bIsValid=false, max=<0 0>, min=<0 0>}
     }
     horizontalBottomSlotImage {
      bIsDynamicallyLoaded=false
      drawAs=NoDrawType
      imageType=FullColor
      mirroring=NoMirror
      resourceName=../../../Engine/Content/Slate/Common/Scrollbar_Background_Horizontal.png
      resourceObject=null
      tiling=NoTile
      imageSize=<8 8>
      margin=<ltrb 0 0 0 0>
      outlineSettings {
       bUseBrushTransparency=false
       roundingType=HalfHeightRadius
       width=0
       color {colorUseRule=UseColor_Specified, specifiedColor=<rgba 0 0 0 0>}
       cornerRadii=<0 0 0 0>
      }
      tintColor {colorUseRule=UseColor_Specified, specifiedColor=<rgba 1 1 1 1>}
      uVRegion {bIsValid=false, max=<0 0>, min=<0 0>}
     }
     horizontalTopSlotImage {
      bIsDynamicallyLoaded=false
      drawAs=NoDrawType
      imageType=FullColor
      mirroring=NoMirror
      resourceName=../../../Engine/Content/Slate/Common/Scrollbar_Background_Horizontal.png
      resourceObject=null
      tiling=NoTile
      imageSize=<8 8>
      margin=<ltrb 0 0 0 0>
      outlineSettings {
       bUseBrushTransparency=false
       roundingType=HalfHeightRadius
       width=0
       color {colorUseRule=UseColor_Specified, specifiedColor=<rgba 0 0 0 0>}
       cornerRadii=<0 0 0 0>
      }
      tintColor {colorUseRule=UseColor_Specified, specifiedColor=<rgba 1 1 1 1>}
      uVRegion {bIsValid=false, max=<0 0>, min=<0 0>}
     }
     hoveredThumbImage {
      bIsDynamicallyLoaded=false
      drawAs=RoundedBox
      imageType=FullColor
      mirroring=NoMirror
      resourceName=../../../Engine/Content/Slate/Common/Scrollbar_Thumb.png
      resourceObject=null
      tiling=NoTile
      imageSize=<12 12>
      margin=<ltrb 0.25 0.25 0.25 0.25>
      outlineSettings {
       bUseBrushTransparency=false
       roundingType=FixedRadius
       width=2
       color {colorUseRule=UseColor_Specified, specifiedColor=<rgba 0 0 0 0>}
       cornerRadii=<6 6 6 6>
      }
      tintColor {colorUseRule=UseColor_Specified, specifiedColor=<rgba 1 1 1 1>}
      uVRegion {bIsValid=false, max=<0 0>, min=<0 0>}
     }
     normalThumbImage {
      bIsDynamicallyLoaded=false
      drawAs=RoundedBox
      imageType=FullColor
      mirroring=NoMirror
      resourceName=../../../Engine/Content/Slate/Common/Scrollbar_Thumb.png
      resourceObject=null
      tiling=NoTile
      imageSize=<12 12>
      margin=<ltrb 0.25 0.25 0.25 0.25>
      outlineSettings {
       bUseBrushTransparency=false
       roundingType=FixedRadius
       width=2
       color {colorUseRule=UseColor_Specified, specifiedColor=<rgba 0 0 0 0>}
       cornerRadii=<6 6 6 6>
      }
      tintColor {colorUseRule=UseColor_Specified, specifiedColor=<rgba 1 1 1 0.75>}
      uVRegion {bIsValid=false, max=<0 0>, min=<0 0>}
     }
     verticalBackgroundImage {
      bIsDynamicallyLoaded=false
      drawAs=Box
      imageType=FullColor
      mirroring=NoMirror
      resourceName=None
      resourceObject=/FortUILibrary/UIKit/ScrollBar/MaterialInstances/MI_UIKit_ScrollbarBackground
      tiling=NoTile
      imageSize=<12 12>
      margin=<ltrb 0.5 0.5 0.5 0.5>
      outlineSettings {
       bUseBrushTransparency=false
       roundingType=HalfHeightRadius
       width=0
       color {colorUseRule=UseColor_Specified, specifiedColor=<rgba 0 0 0 0>}
       cornerRadii=<0 0 0 0>
      }
      tintColor {colorUseRule=UseColor_Specified, specifiedColor=<rgba 1 1 1 1>}
      uVRegion {bIsValid=false, max=<0 0>, min=<0 0>}
     }
     verticalBottomSlotImage {
      bIsDynamicallyLoaded=false
      drawAs=NoDrawType
      imageType=FullColor
      mirroring=NoMirror
      resourceName=../../../Engine/Content/Slate/Common/Scrollbar_Background_Vertical.png
      resourceObject=null
      tiling=NoTile
      imageSize=<8 8>
      margin=<ltrb 0 0 0 0>
      outlineSettings {
       bUseBrushTransparency=false
       roundingType=HalfHeightRadius
       width=0
       color {colorUseRule=UseColor_Specified, specifiedColor=<rgba 0 0 0 0>}
       cornerRadii=<0 0 0 0>
      }
      tintColor {colorUseRule=UseColor_Specified, specifiedColor=<rgba 1 1 1 1>}
      uVRegion {bIsValid=false, max=<0 0>, min=<0 0>}
     }
     verticalTopSlotImage {
      bIsDynamicallyLoaded=false
      drawAs=NoDrawType
      imageType=FullColor
      mirroring=NoMirror
      resourceName=../../../Engine/Content/Slate/Common/Scrollbar_Background_Vertical.png
      resourceObject=null
      tiling=NoTile
      imageSize=<8 8>
      margin=<ltrb 0 0 0 0>
      outlineSettings {
       bUseBrushTransparency=false
       roundingType=HalfHeightRadius
       width=0
       color {colorUseRule=UseColor_Specified, specifiedColor=<rgba 0 0 0 0>}
       cornerRadii=<0 0 0 0>
      }
      tintColor {colorUseRule=UseColor_Specified, specifiedColor=<rgba 1 1 1 1>}
      uVRegion {bIsValid=false, max=<0 0>, min=<0 0>}
     }
    }
   }
   slot {class=GridSlot, properties {padding=<ltrb 15 15 15 15>}}
  }
  LoadingEntriesMessage {
   widgetClassPath=/Compete/TournamentLeaderboard/WBP_LoadingEntriesMessage
   bIsVariable=true
   parent=LiveItemsWrapper
   widget=LoadingEntriesMessage
   properties {loadingMessageText=Loading live game entries, visibility=Collapsed}
   slot {class=GridSlot, properties {horizontalAlignment=HAlign_Center, layer=10, verticalAlignment=VAlign_Bottom, padding=<ltrb 0 0 0 40>}}
  }
  EmptyList {
   widgetClassPath=&GridPanel
   parent=ListSwitcher
   widget=EmptyList
   properties {columnFill [1], rowFill [1]}
   slot {class=CommonVisibilitySwitcherSlot, properties {padding=<ltrb 0 32 0 0>}}
  }
  Image_EmptylIST {
   widgetClassPath=&Image
   bIsVariable=true
   parent=EmptyList
   widget=Image_EmptylIST
   properties {
    brush {
     bIsDynamicallyLoaded=false
     drawAs=Image
     imageType=FullColor
     mirroring=NoMirror
     resourceName=None
     resourceObject=/Compete/MaterialInstances/MI_UI_Compete_TournamentDetailsSection
     tiling=NoTile
     imageSize=<939 32>
     margin=<ltrb 0 0 0 0>
     outlineSettings {
      bUseBrushTransparency=false
      roundingType=FixedRadius
      width=0
      color {colorUseRule=UseColor_Specified, specifiedColor=<rgba 0 0 0 0>}
      cornerRadii=<0.5 0 0 0>
     }
     tintColor {colorUseRule=UseColor_Specified, specifiedColor=<rgba 1 1 1 1>}
     uVRegion {bIsValid=false, max=<0 0>, min=<0 0>}
    }
    colorAndOpacity=<rgba 1 1 1 0.3>
   }
   slot {class=GridSlot}
  }
  EmptyListText {
   widgetClassPath=&CommonTextBlock
   parent=EmptyList
   widget=EmptyListText
   properties {
    autoWrapText=true
    justification=Center
    text="Live matches for top players will be shown here. Check back soon!"
    font {
     bForceMonospaced=false
     bMaterialIsStencil=false
     fontMaterial=null
     fontObject=&HeadingNow
     letterSpacing=0
     monospacedWidth=1
     size=18
     skewAmount=0
     typefaceFontName=75 Medium
     outlineSettings {
      bApplyOutlineToDropShadows=false
      bMiteredCorners=false
      bSeparateFillAlpha=false
      outlineMaterial=null
      outlineSize=0
      outlineColor=<rgba 0 0 0 1>
     }
    }
    shadowOffset=<0 0>
   }
   slot {class=GridSlot, properties {horizontalAlignment=HAlign_Center, verticalAlignment=VAlign_Center}}
  }
  HorizontalBox_34 {
   widgetClassPath=&HorizontalBox
   parent=Content
   widget=HorizontalBox_34
   slot {class=GridSlot, properties {row=3, verticalAlignment=VAlign_Center, padding=<ltrb 0 0 0 48>}}
  }
  WatchButton {
   widgetClassPath=&WBP_UIKit_Button_Regular
   bIsVariable=true
   parent=HorizontalBox_34
   widget=WatchButton
   properties {bIsFocusable=false, minHeight=56, text=Watch, triggeringInputAction {dataTable=&InputActionDataTable, rowName=Confirm}}
   slot {class=HorizontalBoxSlot, properties {verticalAlignment=VAlign_Center, padding=<ltrb 0 0 24 0>, size {sizeRule=Fill, value=1}}}
  }
  CloseButton {
   widgetClassPath=&WBP_UIKit_Button_Regular
   bIsVariable=true
   parent=HorizontalBox_34
   widget=CloseButton
   properties {
    bIsFocusable=false
    buttonType=Quiet
    minHeight=56
    minWidth=482
    text=Close
    triggeringInputAction {dataTable=&InputActionDataTable, rowName=Back}
   }
   slot {class=HorizontalBoxSlot, properties {verticalAlignment=VAlign_Center, size {sizeRule=Fill, value=1}}}
  }
 ]
}
